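# Physics asset for the RoosterMeltMouse pickaxe cosmetic, covering the accordion section of its axe mesh.
# Three capsule bodies sit on the left, middle and right accordion bones, with matching capsule shapes on the two side bones.
# Constraints tie each side body to the middle one with swing and twist locked, so the 45-degree angular limits allow no rotation between them.
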PhysicsAsset /BRCosmetics/Weapons/FORT_Melee/Pickaxe_RoosterMeltMouse/Meshes/RoosterMeltMouse_Axe_Physics
bodies {
 accordion_l_bn [{shapeType=Capsule, length=6, radius=7, shapeName=accordion_l_start_capsule, center=<7.1002 0.0001 0>, rotation=<rot 90 -90 -90.0011>}]
 accordion_middle [{shapeType=Capsule, length=3.1943, radius=6.9255, center=<0 0 6.6555>, rotation=<rot 0 -90.1456 -89.9907>}]
 accordion_r_bn [{shapeType=Capsule, length=6, radius=7, shapeName=accordion_l_start_capsule, center=<7.1002 0.0001 0>, rotation=<rot 90 -90 -90.0011>}]
}
constraints [
 {
  bone1Name=accordion_r_bn
  bone2Name=accordion_middle
  swing1LimitDegrees=45
  swing1Motion=Locked
  swing2LimitDegrees=45
  swing2Motion=Locked
  twistLimitDegrees=45
  twistMotion=Locked
 }
 {
  bone1Name=accordion_l_bn
  bone2Name=accordion_middle
  swing1LimitDegrees=45
  swing1Motion=Locked
  swing2LimitDegrees=45
  swing2Motion=Locked
  twistLimitDegrees=45
  twistMotion=Locked
 }
]
